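# Niagara idle effect for the SquidGlisten Lift pickaxe: persistent glow rings and recurring, mostly blue lightning at both ends of the weapon.
# Gameplay adds a deploy burst of flash sprites, lights and lightning meshes after a system delay, while front end suppresses it.
# Center and side lightning meshes take fixed offsets per partition and index variants through User.LightningOrder.
# System age fades in rings and center lightning; a time-of-day scalar dims the flash.
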NiagaraSystem /BRCosmetics/Weapons/FORT_Melee/Pickaxe_SquidGlisten_Lift/FX/NS_Pickaxe_SquidGlisten_Lift_Idle
dependencies {
 usedDataInterfaces [
  &NiagaraDataInterfaceCurve
  &NiagaraDataInterfaceArrayFloat3
  &NiagaraDataInterfaceArrayFloat
  &NiagaraDataInterfaceMeshRendererInfo
  &NiagaraDataInterfaceColorCurve
  &NiagaraDataInterfaceVectorCurve
  &NiagaraDataInterfaceSkeletalMesh
 ]
 usedDynamicInputs [
  &SimulationPosition
  &Add_Integer
  &Multiply_Int
  &RandomRangeInt
  &Add_Vector
  /Niagara/DynamicInputs/Helpers/SimulationPosition
  &UniformRangedVector
  &FNTimeOfDayScalar
 ]
 usedModules [
  &SystemState
  &self:SystemSpawnScript.SystemScriptSource.SystemScriptGraph.NiagaraNodeAssignment_1.SetVariables_#1
  &self:SystemSpawnScript.SystemScriptSource.SystemScriptGraph.NiagaraNodeAssignment_2.SetVariables_#2
  &self:SystemSpawnScript.SystemScriptSource.SystemScriptGraph.NiagaraNodeAssignment_5.SetVariables_#3
 ]
}
emitters {
 FlashDeploy {
  emitterUpdateScript [
   &EmitterState(Life Cycle Mode=::NewEnumerator1, Loop Behavior=::NewEnumerator1, Loop Duration=2, UseLoopDelay=false)
   &SpawnBurst_Instantaneous(
    Spawn Count=2
    Spawn Probability=&FN_FrontEndFloatModifier(Gameplay=1, Front End=0)
    Spawn Time=$System.DeployDelay
    Use Spawn Probability=true
   )
  ]
  particleSpawnScript [
   &PartitionParticles()
   &InitializeParticle(
    Lifetime=0.4
    Mass Mode=::NewEnumerator1
    Mesh Scale Mode=::NewEnumerator3
    Position Offset=&SelectVectorFromArray(
     Direct Array Index=$Particles.PartitionParticles.Partition
     Vector Selection Array=DataInterface() {propertyValues {FloatData [<0 8 80>, <0 8 -30>]}}
    )
    Sprite Size=<100 250>
    Sprite Size Mode=::NewEnumerator4
    UsePositionOffset=true
    Write Material Random=true
   ) {module=InitializeParticle001}
   &CameraOffset(Camera Offset Amount=10)
   &SpriteFacingAndAlignment(Sprite Alignment=<0 0 1>, WriteAlignment=true, WriteFacing=false)
  ]
  particleUpdateScript [
   &SolveForcesAndVelocity(Clamp Velocity=false, Limit Acceleration=false) {module=SolveForcesAndVelocity001}
   &SolveForcesAndVelocity(Clamp Velocity=false, Limit Acceleration=false)
   &ParticleState()
   &ScaleSpriteSize(Uniform Curve Sprite Scale=DataInterface() {propertyValues {Curve {keys [
        {interpMode=RCIM_Cubic, leaveTangentWeight=0.504, tangentMode=RCTM_User, time=0, value=0}
        {arriveTangentWeight=0.1, interpMode=RCIM_Cubic, leaveTangentWeight=0.2333, tangentMode=RCTM_User, time=0.5, value=1}
        {tangentMode=RCTM_User, time=1, value=1}
       ]}}})
   &Color(
    Color=<rgba 0.125 0.5208 2.5 0.18>
    Scale Alpha=&FloatFromCurve(FloatCurve=DataInterface() {propertyValues {Curve {keys [
         {arriveTangent=-2.5644, interpMode=RCIM_Cubic, leaveTangent=-2.5644, tangentMode=RCTM_User, time=0, value=1}
         {interpMode=RCIM_Cubic, tangentMode=RCTM_User, time=1, value=0}
        ]}}})
    Scale Color=&VectorFromCurve(VectorCurve=DataInterface() {propertyValues {
       XCurve {keys [
         {arriveTangent=-94.5365, interpMode=RCIM_Cubic, leaveTangent=-94.5365, tangentMode=RCTM_User, time=0.7, value=12}
         {arriveTangent=-1.0855, interpMode=RCIM_Cubic, leaveTangent=-1.0855, tangentMode=RCTM_User, time=1, value=1}
        ]}
       YCurve {keys [
         {arriveTangent=-94.5365, interpMode=RCIM_Cubic, leaveTangent=-94.5365, tangentMode=RCTM_User, time=0.7, value=12}
         {arriveTangent=-1.0855, interpMode=RCIM_Cubic, leaveTangent=-1.0855, tangentMode=RCTM_User, time=1, value=1}
        ]}
       ZCurve {keys [
         {arriveTangent=-94.5365, interpMode=RCIM_Cubic, leaveTangent=-94.5365, tangentMode=RCTM_User, time=0.7, value=12}
         {arriveTangent=-1.0855, interpMode=RCIM_Cubic, leaveTangent=-1.0855, tangentMode=RCTM_User, time=1, value=1}
        ]}
      }})
   )
   &FN_Sprite_Attributes(
    Create SpriteSize Attribute=true
    SpriteSize=&Multiply_Vector2D(
     A=<30 250>
     B=&Vector2DFromFloat(Value=&FloatFromCurve(FloatCurve=DataInterface() {propertyValues {Curve {keys [<at 0 0 RCIM_Cubic>,<at 0.5 1 RCIM_Cubic>]}}}))
    )
    Write to Module Namespace=true
   ) {module=FN_Sprite}
   /SetVariables(Particles.SpriteColor=&MultiplyLinearColor(
     Linear Color=$Particles.Color
     Scale Factor=&MakeVector4(
      X=$System.SMTOD
      Y=$System.SMTOD
      Z=$System.SMTOD
      W=&MakeCustomFloatFromBool(Bool=$User.bIsFrontEndPreview, True Float=0.06, False Float=0.13)
     )
    )) {isSetParameters=true}
   &DynamicMaterialParameters(Write Parameter Index 0=false)
   &FN_Light_Attributes(
    Color=<rgba 0.1 0.175 1 1>
    Create Color Attribute=true
    Create Exponent Attribute=true
    Create Position Attribute=true
    Create Radius Attribute=true
    Enable Particle Light Enable=true
    Exponent=3
    Light Enabled=true
    Position:NiagaraPosition=$Particles.Position
    Radius=50
    Scale Color=&FloatFromCurve(
     FloatCurve=DataInterface() {propertyValues {Curve {keys [
         {arriveTangent=-3.9329, interpMode=RCIM_Cubic, leaveTangent=-3.9329, tangentMode=RCTM_User, time=0, value=1}
         {arriveTangent=-0.3589, tangentMode=RCTM_User, time=1, value=0}
        ]}}}
     Scale Curve=15
    )
   )
   &ScaleColor(Scale RGB=&VectorFromFloat(Value=$System.SMTOD), ScaleA=true, ScaleRGB=true) {module=ScaleColor002}
  ]
  rendererClasses [&NiagaraLightRendererProperties, &NiagaraSpriteRendererProperties]
  renderers [
   {rendererClass=&NiagaraLightRendererProperties, rendererIndex=0}
   {rendererClass=&NiagaraSpriteRendererProperties, rendererIndex=1}
   {rendererClass=&NiagaraSpriteRendererProperties, rendererIndex=2}
  ]
 }
 GlowRings {
  emitterUpdateScript [
   &EmitterState(
    Inactive Response=::NewEnumerator1
    Life Cycle Mode=::NewEnumerator1
    Loop Behavior=::NewEnumerator1
    Loop Duration Mode=::NewEnumerator1
    Scalability Mode=::NewEnumerator1
    UseLoopDelay=false
   )
   &SpawnBurst_Instantaneous(Spawn Count=12, Use Spawn Probability=false, UseLoopCountLimit=false)
  ]
  particleSpawnScript [
   &FN_SetSpawnGroup()
   &PartitionParticles(Number Of Partitions=12, Particles Per Partition=1, Partition Mode:ENiagara_PartitionMode_v2=::NewEnumerator1)
   &InitializeParticle(
    Color=<rgba 0.683 1.0428 5 1>
    Position Offset=&SelectVectorFromArray(
     Direct Array Index=$Particles.PartitionParticles.Partition
     Vector Selection Array=DataInterface() {propertyValues {FloatData [
        <1.1 3.8 -65.5>
        <1 4.23 -71>
        <1.05 10.3 -65.5>
        <1 9.95 -71>
        <1.1 3.8 105.5>
        <1 4.23 111>
        <1.05 10.3 105.5>
        <1 9.95 111>
        <4 7.1 -67.8>
        <-2 7.1 -67.8>
        <4 7.1 107.8>
        <-2 7.1 107.8>
       ]}}
    )
    Sprite Size Mode=::NewEnumerator3
    Uniform Sprite Size=&SelectFloatFromArray(
     Direct Array Index=$Particles.PartitionParticles.Partition
     Float Selection Array=DataInterface() {propertyValues {FloatData [2.3,1.9,2.3,1.9,2.3,1.9,2.3,1.9,1,1,1,1]}}
    )
    UsePositionOffset=true
    Write Material Random=true
   )
  ]
  particleUpdateScript [
   &ParticleState(Kill Particles When Lifetime Has Elapsed=false, Let Infinitely Lived Particles Die When Emitter Deactivates=true)
   &ScaleColor(
    Scale Alpha=&FN_FrontEndFloatModifier(
     Gameplay=&FloatFromCurve(
      FloatCurve=DataInterface() {propertyValues {Curve {keys [
          {arriveTangent=-2.8571, interpMode=RCIM_Cubic, time=0.5, value=0}
          {interpMode=RCIM_Cubic, tangentMode=RCTM_User, time=0.6, value=2}
          <at 0.95 1>
         ]}}}
      CurveIndex=$System.Age
     )
     Front End=&FloatFromCurve(FloatCurve=DataInterface() {propertyValues {Curve {keys [<at 0.2 0 RCIM_Cubic>,<at 1 1>]}}}, CurveIndex=$System.Age)
    )
    Scale RGB=&VectorFromFloat(Value=&FloatFromCurve(
      FloatCurve=DataInterface() {propertyValues {Curve {keys [
          {arriveTangent=-5.7143, interpMode=RCIM_Cubic, time=0.5, value=0}
          {interpMode=RCIM_Cubic, tangentMode=RCTM_User, time=0.6, value=3}
          <at 0.95 1>
         ]}}}
      CurveIndex=$System.Age
     ))
    ScaleA=true
    ScaleRGB=true
   )
   &SpriteFacingAndAlignment(
    Sprite Facing=&SelectVectorFromArray(
     Direct Array Index=$Particles.PartitionParticles.Partition
     Vector Selection Array=DataInterface() {propertyValues {FloatData [<0 -1 0>, <0 -1 0>, <0 1 0>, <0 1 0>, <0 -1 0>, <0 -1 0>, <0 1 0>, <0 1 0>, <1 0 0>, <-1 0 0>, <1 0 0>, <-1 0 0>]}}
    )
    WriteAlignment=false
    WriteFacing=true
   )
   /SetVariables(
    Particles.SpriteGlowCameraOffset=25
    Particles.SpriteGlowColor:LinearColor=&Multiply_LinearColorByFloat(Linear Color=$Particles.Color, Float=0.5)
    Particles.SpriteGlowSize:Vector2f=&Multiply_Vector2D_ByFloat(Float=8, Vector2D=$Particles.SpriteSize)
   ) {isSetParameters=true}
  ]
  rendererClasses [&NiagaraSpriteRendererProperties]
  renderers [
   {rendererClass=&NiagaraSpriteRendererProperties}
   {rendererClass=&NiagaraSpriteRendererProperties, rendererIndex=1}
  ]
 }
 LightningCenter {
  emitterUpdateScript [
   &EmitterState(
    Inactive Response=::NewEnumerator1
    Life Cycle Mode=::NewEnumerator1
    Loop Behavior=::NewEnumerator1
    Loop Duration Mode=::NewEnumerator1
    UseLoopDelay=false
   )
   &SpawnRate(
    Spawn Probability=&FN_FrontEndFloatModifier(
     Gameplay=&FloatFromCurve(FloatCurve=DataInterface() {propertyValues {Curve {keys [<at 0.6 0 RCIM_Cubic>,<at 0.7 1>]}}}, CurveIndex=$System.Age)
     Front End=1
    )
    SpawnRate=12
    Use Spawn Probability=true
   )
   &SpawnBurst_Instantaneous(
    Spawn Count=12
    Spawn Probability=&FN_FrontEndFloatModifier(Gameplay=1, Front End=0)
    Spawn Time=0.5
    Use Spawn Probability=true
    UseLoopCountLimit=false
   ) {module=SpawnBurst_Instantaneous001}
  ]
  particleSpawnScript [
   &PartitionParticles(Number Of Partitions=4, Partition Mode:ENiagara_PartitionMode_v2=::NewEnumerator5)
   &InitializeParticle(
    Color=<rgba 3 3 3 1>
    Lifetime=&RandomRangeFloat(Minimum=0.8, Maximum=1.8)
    Mesh Scale=<0.5 0.5 0.5>
    Mesh Scale Mode=::NewEnumerator4
    Position Offset=&SelectVectorFromArray(
     Direct Array Index=$Particles.PartitionParticles.Partition
     Vector Selection Array=DataInterface() {propertyValues {FloatData [<4.5 6.5 -67.5>, <-2 6.5 -67.5>, <4.5 7 108>, <-2 6.5 108.5>]}}
    )
    UsePositionOffset=true
    Write Material Random=true
   )
   /SetVariables(
    Particles.InvertLightning:NiagaraBool=&RandomBool()
    Particles.LightningIndex:NiagaraInt32=&ModuloInt(
     A=&SelectIntFromArray(
      Direct Array Index=&ReturnParticleIDAttributes()
      Direct Array Mode=::NewEnumerator1
      Int Selection Array=$User.LightningOrder
     )
     B=4
    )
    Particles.MirrorLightning:NiagaraBool=&RandomBool()
   ) {isSetParameters=true, module=SetVariables_#4}
   /SetVariables(Particles.Scale=&Multiply_Vector(A=<0.4 0.275 0.5>, B=<1.25 1 1>)) {isSetParameters=true, module=SetVariables_#5}
   &InitialMeshOrientation(
    Mesh Orientation Mode=::NewEnumerator2
    Orientation Vector=&SelectVectorFromArray(
     Direct Array Index=$Particles.PartitionParticles.Partition
     Vector Selection Array=DataInterface() {propertyValues {FloatData [<1 0 0>, <-1 0 0>, <1 0 0>, <-1 0 0>]}}
    )
    Rotation=&SelectVectorFromArray(
     Array Sampling Mode=::NewEnumerator0
     Direct Array Index=?
     Vector Selection Array=DataInterface() {propertyValues {FloatData [<0.6 -0.2 -0.1>, <0.4 -0.2 0>, <0.05 -0.05 0.2>, <-0.12 0.05 0.2>]}}
     Random Seed=0
    )
    Use Rotation Vector=true
   )
  ]
  particleUpdateScript [
   &DynamicMaterialParameters(
    Index 0 Param 1=&FloatFromCurve(
     FloatCurve=DataInterface() {propertyValues {Curve {keys [{tangentMode=RCTM_User, time=0, value=0}, {tangentMode=RCTM_User, time=1, value=1}]}}}
     Scale Curve=3
    )
    Index 0 Param 2=0.7
    Index 0 Param 3=&FloatFromCurve(FloatCurve=DataInterface() {propertyValues {Curve {keys [<at 0 0 RCIM_Cubic>, {arriveTangent=0.2057, interpMode=RCIM_Cubic, leaveTangent=0.2057, tangentMode=RCTM_User, time=1, value=3}]}}})
    Index 0 Param 4=2
    Index 1 Param 1=0.3
    Index 1 Param 2=1
    Index 1 Param 3=&Lerp_Float(
     A=0.75
     B=1.5
     Alpha=&SelectFloatFromArray(
      Array Sampling Mode=::NewEnumerator0
      Direct Array Index=?
      Float Selection Array=DataInterface() {propertyValues {FloatData [0,1]}}
      Random Seed=0
     )
    )
    Index 1 Param 4=1
    Module.Index0Float0Write=true
    Module.Index0Float1Write=true
    Module.Index0Float2Write=true
    Module.Index0Float3Write=true
    Module.Index1Float0Write=true
    Module.Index1Float1Write=true
    Module.Index1Float2Write=true
    Module.Index1Float3Write=true
    Write Parameter Index 1=::NewEnumerator1
   )
   /SetVariables(Particles.Glow Dy03:Vector4f=&MakeVector4(
     X=&FloatFromCurve(FloatCurve=DataInterface() {propertyValues {Curve {keys [<at 0.5 1>,<at 1 0.3>]}}}, Scale Curve=0.3)
     Y=1
     Z=&Lerp_Float(A=1, B=2, Alpha=&RandomRangeFloat(Minimum=0, Maximum=1))
     W=1
    )) {isSetParameters=true, module=SetVariables_#6}
   /SetVariables(Particles.Glow Dy02:Vector4f=&MakeVector4(
     X=&FloatFromCurve(FloatCurve=DataInterface() {propertyValues {Curve {keys [
          {interpMode=RCIM_Cubic, tangentMode=RCTM_User, time=0, value=0.2}
          {arriveTangent=4.8898, interpMode=RCIM_Cubic, leaveTangent=4.8898, tangentMode=RCTM_User, time=0.5, value=1}
         ]}}})
     Y=0.7
     Z=0
     W=2
    )) {isSetParameters=true, module=SetVariables_#7}
   /SetVariables(Particles.Lightning glow color:LinearColor=&MakeLinearColorFromVectorAndFloat(
     "Vector (RGB)"=&MakeVectorFromLinearColorRGB(Color=&MultiplyLinearColor(
       Linear Color=<rgba 0.1979 0.4983 2 1>
       Scale Factor=&MakeVector4_FromFloat(Float=&RandomRangeFloat(Minimum=1, Maximum=1, Evaluation Type=::NewEnumerator1))
      ))
     "Float (Alpha)"=&Multiply_Float(
      A=&FloatFromCurve(FloatCurve=DataInterface() {propertyValues {Curve {keys [<at 0 1 RCIM_Cubic>, {interpMode=RCIM_Cubic, leaveTangent=-3.6307, time=0.9, value=0}]}}})
      B=0.75
     )
    )) {isSetParameters=true, module=SetVariables_#8}
   &ParticleState()
   &Color(
    Color=&ColorFromCurve(ColorCurve=DataInterface() {propertyValues {
       AlphaCurve {keys [<at 0 0>, {arriveTangent=0.3873, interpMode=RCIM_Cubic, leaveTangent=0.3873, time=0.0998, value=1}, <at 0.7746 0.3>, <at 1 0 RCIM_Cubic>]}
       BlueCurve {keys [<at 0 10 RCIM_Cubic>,<at 0.2919 3>,<at 1 1 RCIM_Cubic>]}
       GreenCurve {keys [<at 0 3.1845 RCIM_Cubic>,<at 0.2919 0.25>,<at 1 0.0833 RCIM_Cubic>]}
       RedCurve {keys [<at 0 1.8214 RCIM_Cubic>,<at 0.2919 0>,<at 1 0 RCIM_Cubic>]}
      }})
    Scale Alpha=&FloatFromCurve(FloatCurve=DataInterface() {propertyValues {Curve {keys [<at 0.6 0 RCIM_Cubic>,<at 0.8 1>]}}}, CurveIndex=$System.Age)
   )
  ]
  rendererClasses [&NiagaraMeshRendererProperties]
  renderers [
   {rendererClass=&NiagaraMeshRendererProperties, rendererIndex=0}
   {rendererClass=&NiagaraMeshRendererProperties, rendererIndex=1}
  ]
 }
 LightningDeploy {
  emitterUpdateScript [
   &EmitterState(Life Cycle Mode=::NewEnumerator1, Loop Behavior=::NewEnumerator1, Loop Duration=2, UseLoopDelay=false)
   &SpawnBurst_Instantaneous(
    Spawn Count=6
    Spawn Probability=&FN_FrontEndFloatModifier(Gameplay=1, Front End=0)
    Spawn Time=$System.DeployDelay
    Use Spawn Probability=true
    UseLoopCountLimit=false
   )
  ]
  particleSpawnScript [
   &PartitionParticles()
   &CameraOffset(Camera Offset Amount=2)
   &InitializeParticle(
    Color=<rgba 10 10 10 1>
    Lifetime=&RandomRangeFloat(Minimum=0.6, Maximum=1.2)
    Mesh Scale=&Multiply_Vector(
     A=&RandomRangeVector(Minimum=<1 1 2>, Maximum=<1.3 1.3 2>)
     B=&SelectVectorFromArray(
      Array Sampling Mode=::NewEnumerator0
      Direct Array Index=?
      Vector Selection Array=DataInterface() {propertyValues {FloatData [<1 1 1.1>, <1.2 1.2 1.12>, <0.7 0.7 1.1>]}}
      Random Seed=0
     )
    )
    Mesh Scale Mode=::NewEnumerator4
    Position Mode=::NewEnumerator0
    Sprite Size Mode=::NewEnumerator1
    Uniform Sprite Size Max=50
    Uniform Sprite Size Min=35
    Write Material Random=true
   )
   &InitialMeshOrientation(
    Mesh Axis To Orient=<0 1 0>
    Mesh Orientation Mode=::NewEnumerator2
    Orientation Vector=<0 0 1>
    Rotation=&RandomRangeVector(Minimum=<0 0 0>, Maximum=<0 1 0>)
    Use Rotation Vector=true
   )
   &ShapeLocation(
    MaskSpawnGroup=false
    Offset=&Lerp_Vector(
     A=&RandomRangeVector(Minimum=<0 8 60>, Maximum=<0 8 65>)
     B=&RandomRangeVector(Minimum=<0 8 -18>, Maximum=<0 8 -23>)
     Alpha=&MakeFloatFromInt(Int=$Particles.PartitionParticles.Partition)
    )
    Offset Mode=::NewEnumerator0
    Override Seed=false
    Sphere Radius=0
    UseOwnerScale=true
   ) {module=ShapeLocation001}
  ]
  particleUpdateScript [
   &ParticleState()
   &ScaleMeshSize(Scale Factor=&Lerp_Vector(A=<0.6 0.78 0.6>, B=<0.8 0.78 0.8>, Alpha=&RandomRangeFloat(Minimum=0, Maximum=1))) {module=ScaleMeshSize001}
   &ScaleMeshSize(Scale Factor=&MakeVector(
     X=1
     Y=&FloatFromCurve(
      FloatCurve=DataInterface() {propertyValues {Curve {keys [
          {interpMode=RCIM_Cubic, leaveTangent=0.1844, tangentMode=RCTM_Break, time=0, value=1}
          {interpMode=RCIM_Cubic, leaveTangent=-0.1745, time=0.7665, value=1.07}
         ]}}}
      Scale Curve=&SelectFloatFromArray(Direct Array Index=$Particles.PartitionParticles.Partition, Float Selection Array=DataInterface() {propertyValues {FloatData [2,-2]}})
     )
     Z=1
    ))
   &Color(
    Color=&ColorFromCurve(ColorCurve=DataInterface() {propertyValues {
       AlphaCurve {keys [<at 0 0>, {arriveTangent=0.3873, interpMode=RCIM_Cubic, leaveTangent=0.3873, time=0.0998, value=1}, <at 0.7746 0.3>, <at 1 0 RCIM_Cubic>]}
       BlueCurve {keys [<at 0 10 RCIM_Cubic>,<at 0.2919 3>,<at 1 1 RCIM_Cubic>]}
       GreenCurve {keys [<at 0 3.1845 RCIM_Cubic>,<at 0.2919 0.25>,<at 1 0.0833 RCIM_Cubic>]}
       RedCurve {keys [<at 0 1.8214 RCIM_Cubic>,<at 0.2919 0>,<at 1 0 RCIM_Cubic>]}
      }})
    Scale Color=<2 2 2>
   )
   &DynamicMaterialParameters(
    Index 0 Param 1=&FloatFromCurve(FloatCurve=DataInterface() {propertyValues {Curve {keys [<at 0 0>,<at 1 1>]}}}, Scale Curve=3)
    Index 0 Param 2=0.7
    Index 0 Param 3=&FloatFromCurve(FloatCurve=DataInterface() {propertyValues {Curve {keys [<at 0 0 RCIM_Cubic>, {arriveTangent=0.2057, interpMode=RCIM_Cubic, leaveTangent=0.2057, tangentMode=RCTM_User, time=1, value=3}]}}})
    Index 0 Param 4=1
    Index 1 Param 1=0.6
    Index 1 Param 2=&RandomRangeFloat(Minimum=0, Maximum=1)
    Index 1 Param 3=&Lerp_Float(
     A=1.25
     B=2.25
     Alpha=&SelectFloatFromArray(
      Array Sampling Mode=::NewEnumerator0
      Direct Array Index=?
      Float Selection Array=DataInterface() {propertyValues {FloatData [0,1]}}
      Random Seed=0
     )
    )
    Index 1 Param 4=1
    Module.Index0Float0Write=true
    Module.Index0Float1Write=true
    Module.Index0Float2Write=true
    Module.Index0Float3Write=true
    Module.Index1Float0Write=true
    Module.Index1Float1Write=true
    Module.Index1Float2Write=true
    Module.Index1Float3Write=true
    Write Parameter Index 1=::NewEnumerator1
   )
   /SetVariables(Particles.Lightning glow color:LinearColor=&MakeLinearColorFromVectorAndFloat(
     "Vector (RGB)"=&MakeVectorFromLinearColorRGB(Color=&MultiplyLinearColor(
       Linear Color=<rgba 0.1979 0.4983 2 1>
       Scale Factor=&MakeVector4_FromFloat(Float=&RandomRangeFloat(Minimum=1, Maximum=1, Evaluation Type=::NewEnumerator1))
      ))
     "Float (Alpha)"=&Multiply_Float(
      A=&FloatFromCurve(FloatCurve=DataInterface() {propertyValues {Curve {keys [<at 0 1 RCIM_Cubic>, {interpMode=RCIM_Cubic, leaveTangent=-3.6307, time=0.9, value=0}]}}})
      B=0.7
     )
    )) {isSetParameters=true}
   /SetVariables(Particles.Glow Dy02:Vector4f=&MakeVector4(
     X=&FloatFromCurve(FloatCurve=DataInterface() {propertyValues {Curve {keys [
          {interpMode=RCIM_Cubic, tangentMode=RCTM_User, time=0, value=0.2}
          {arriveTangent=4.8898, interpMode=RCIM_Cubic, leaveTangent=4.8898, tangentMode=RCTM_User, time=0.3, value=1}
         ]}}})
     Y=0.7
     Z=0
     W=1
    )) {isSetParameters=true}
   /SetVariables(Particles.Glow Dy03:Vector4f=&MakeVector4(
     X=&FloatFromCurve(FloatCurve=DataInterface() {propertyValues {Curve {keys [
          {arriveTangent=-3.6512, interpMode=RCIM_Cubic, leaveTangent=-3.6512, tangentMode=RCTM_User, time=0.5, value=1}
          {arriveTangent=-0.3493, interpMode=RCIM_Cubic, leaveTangent=-0.3493, tangentMode=RCTM_User, time=1, value=0.3}
         ]}}})
     Y=1
     Z=&Lerp_Float(A=1, B=2, Alpha=&RandomRangeFloat(Minimum=0, Maximum=1))
     W=1
    )) {isSetParameters=true}
   &SolveForcesAndVelocity(Clamp Velocity=false, Limit Acceleration=false)
  ]
  rendererClasses [&NiagaraMeshRendererProperties]
  renderers [
   {rendererClass=&NiagaraMeshRendererProperties, rendererIndex=0}
   {rendererClass=&NiagaraMeshRendererProperties, rendererIndex=1}
  ]
 }
 LightningSides {
  emitterUpdateScript [
   &EmitterState(
    Inactive Response=::NewEnumerator1
    Life Cycle Mode=::NewEnumerator1
    Loop Behavior=::NewEnumerator1
    Loop Duration Mode=::NewEnumerator1
    UseLoopDelay=false
   )
   &SpawnRate(
    Spawn Probability=&FN_FrontEndFloatModifier(
     Gameplay=&FloatFromCurve(FloatCurve=DataInterface() {propertyValues {Curve {keys [<at 0.6 0 RCIM_Cubic>,<at 0.7 1>]}}}, CurveIndex=$System.Age)
     Front End=1
    )
    SpawnRate=12
    Use Spawn Probability=true
   )
   &SpawnBurst_Instantaneous(
    Spawn Count=12
    Spawn Probability=&FN_FrontEndFloatModifier(Gameplay=1, Front End=0)
    Spawn Time=0.5
    Use Spawn Probability=true
    UseLoopCountLimit=false
   )
  ]
  particleSpawnScript [
   &PartitionParticles(Number Of Partitions=4, Partition Mode:ENiagara_PartitionMode_v2=::NewEnumerator5)
   &InitializeParticle(
    Color=<rgba 3 3 3 1>
    Lifetime=&RandomRangeFloat(Minimum=0.8, Maximum=1.8)
    Mesh Scale=<0.5 0.5 0.5>
    Mesh Scale Mode=::NewEnumerator4
    Position Offset=&SelectVectorFromArray(
     Direct Array Index=$Particles.PartitionParticles.Partition
     Vector Selection Array=DataInterface() {propertyValues {FloatData [<1 7 -68.3>, <1 7 108>, <1.5 7 108.1>, <1.5 7 -68.1>]}}
    )
    UsePositionOffset=true
    Write Material Random=true
   )
   /SetVariables(
    Particles.InvertLightning:NiagaraBool=&RandomBool()
    Particles.LightningIndex:NiagaraInt32=&ModuloInt(
     A=&SelectIntFromArray(
      Direct Array Index=&ReturnParticleIDAttributes()
      Direct Array Mode=::NewEnumerator1
      Int Selection Array=$User.LightningOrder
     )
     B=4
    )
    Particles.MirrorLightning:NiagaraBool=&RandomBool()
   ) {isSetParameters=true, module=SetVariables_#4}
   /SetVariables(Particles.Scale=&Multiply_Vector(
     A=<0.35 0.3 0.5>
     B=&Multiply_Vector(
      A=&SelectVectorFromArray(
       Direct Array Index=$Particles.PartitionParticles.Partition
       Vector Selection Array=DataInterface() {propertyValues {FloatData [<1 1 1>, <1 -1 1>, <-1 1 1>, <-1 -1 1>]}}
      )
      B=&SelectVectorFromArray(
       Array Sampling Mode=::NewEnumerator0
       Direct Array Index=$Particles.PartitionParticles.Partition
       Vector Selection Array=DataInterface() {propertyValues {FloatData [<1 1 1>, <1 -1 -1>]}}
       Random Seed=0
      )
     )
    )) {isSetParameters=true, module=SetVariables_#5}
   &InitialMeshOrientation(
    Mesh Axis To Orient=<0 0 1>
    Mesh Orientation Mode=::NewEnumerator2
    Orientation Vector=<0 1 0>
    Rotation=<0 0.25 0>
    Use Rotation Vector=true
   )
  ]
  particleUpdateScript [
   &DynamicMaterialParameters(
    Index 0 Param 1=&FloatFromCurve(
     FloatCurve=DataInterface() {propertyValues {Curve {keys [{tangentMode=RCTM_User, time=0, value=0}, {tangentMode=RCTM_User, time=1, value=1}]}}}
     Scale Curve=3
    )
    Index 0 Param 2=0.7
    Index 0 Param 3=&FloatFromCurve(FloatCurve=DataInterface() {propertyValues {Curve {keys [<at 0 0 RCIM_Cubic>, {arriveTangent=0.2057, interpMode=RCIM_Cubic, leaveTangent=0.2057, tangentMode=RCTM_User, time=1, value=3}]}}})
    Index 0 Param 4=2
    Index 1 Param 1=0.3
    Index 1 Param 2=1
    Index 1 Param 3=&Lerp_Float(
     A=0.75
     B=1.5
     Alpha=&SelectFloatFromArray(
      Array Sampling Mode=::NewEnumerator0
      Direct Array Index=?
      Float Selection Array=DataInterface() {propertyValues {FloatData [0,1]}}
      Random Seed=0
     )
    )
    Index 1 Param 4=1
    Module.Index0Float0Write=true
    Module.Index0Float1Write=true
    Module.Index0Float2Write=true
    Module.Index0Float3Write=true
    Module.Index1Float0Write=true
    Module.Index1Float1Write=true
    Module.Index1Float2Write=true
    Module.Index1Float3Write=true
    Write Parameter Index 1=::NewEnumerator1
   )
   /SetVariables(Particles.Glow Dy03:Vector4f=&MakeVector4(
     X=&FloatFromCurve(FloatCurve=DataInterface() {propertyValues {Curve {keys [<at 0.5 1>,<at 1 0.3>]}}}, Scale Curve=0.6)
     Y=1
     Z=&Lerp_Float(A=1, B=2, Alpha=&RandomRangeFloat(Minimum=0, Maximum=1))
     W=1
    )) {isSetParameters=true, module=SetVariables_#6}
   /SetVariables(Particles.Glow Dy02:Vector4f=&MakeVector4(
     X=&FloatFromCurve(FloatCurve=DataInterface() {propertyValues {Curve {keys [
          {interpMode=RCIM_Cubic, tangentMode=RCTM_User, time=0, value=0.2}
          {arriveTangent=4.8898, interpMode=RCIM_Cubic, leaveTangent=4.8898, tangentMode=RCTM_User, time=0.5, value=1}
         ]}}})
     Y=0.7
     Z=0
     W=1
    )) {isSetParameters=true, module=SetVariables_#7}
   /SetVariables(Particles.Lightning glow color:LinearColor=&MakeLinearColorFromVectorAndFloat(
     "Vector (RGB)"=&MakeVectorFromLinearColorRGB(Color=&MultiplyLinearColor(
       Linear Color=<rgba 0.1979 0.4983 2 1>
       Scale Factor=&MakeVector4_FromFloat(Float=&RandomRangeFloat(Minimum=1, Maximum=1, Evaluation Type=::NewEnumerator1))
      ))
     "Float (Alpha)"=&Multiply_Float(
      A=&FloatFromCurve(FloatCurve=DataInterface() {propertyValues {Curve {keys [<at 0 1 RCIM_Cubic>, {interpMode=RCIM_Cubic, leaveTangent=-3.6307, time=0.9, value=0}]}}})
      B=0.75
     )
    )) {isSetParameters=true, module=SetVariables_#8}
   &ParticleState()
   &Color(
    Color=&ColorFromCurve(ColorCurve=DataInterface() {propertyValues {
       AlphaCurve {keys [<at 0 0>, {arriveTangent=0.3873, interpMode=RCIM_Cubic, leaveTangent=0.3873, time=0.0998, value=1}, <at 0.7746 0.3>, <at 1 0 RCIM_Cubic>]}
       BlueCurve {keys [<at 0 10 RCIM_Cubic>,<at 0.2919 3>,<at 1 1 RCIM_Cubic>]}
       GreenCurve {keys [<at 0 3.1845 RCIM_Cubic>,<at 0.2919 0.25>,<at 1 0.0833 RCIM_Cubic>]}
       RedCurve {keys [<at 0 1.8214 RCIM_Cubic>,<at 0.2919 0>,<at 1 0 RCIM_Cubic>]}
      }})
    Scale Color=<2 2 2>
   )
   &SpriteFacingAndAlignment(Sprite Facing=<0 1 0>, WriteAlignment=false, WriteFacing=true)
  ]
  rendererClasses [&NiagaraMeshRendererProperties]
  renderers [
   {rendererClass=&NiagaraMeshRendererProperties, rendererIndex=0}
   {rendererClass=&NiagaraMeshRendererProperties, rendererIndex=1}
  ]
 }
}
summary {
 systemName=NS_Pickaxe_SquidGlisten_Lift_Idle
 emitters [
  {emitterName=LightningSides, rendererClasses [&NiagaraMeshRendererProperties]}
  {emitterName=LightningCenter, rendererClasses [&NiagaraMeshRendererProperties]}
  {emitterName=GlowRings, rendererClasses [&NiagaraSpriteRendererProperties]}
  {emitterName=FlashDeploy, rendererClasses [&NiagaraLightRendererProperties, &NiagaraSpriteRendererProperties]}
  {emitterName=LightningDeploy, rendererClasses [&NiagaraMeshRendererProperties]}
 ]
 userVariables [
  User.bIsFrontEnd {type=NiagaraBool}
  User.bIsFrontEndPreview {type=NiagaraBool}
  User.LightningOrder {
   type=NiagaraDataInterfaceArrayInt32
   description=""
   defaultValue {
    struct=NiagaraExt_VariableValue_DataInterface
    value {
     dataInterface=&self:NiagaraDataInterfaceArrayInt32_3
     dataInterfaceClass=&NiagaraDataInterfaceArrayInt32
    }
   }
  }
 ]
}
